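# Quest rewards panel widget for S39 MimosaShark, promoting Shiny Llama drops in Squads.
# An Overlay layers glow and particle materials and three placeholder images beneath a squad-shared progress notice and icon-decorated rich-text description.
# SetMobileStyle checks IsUsingCompactUI, centers the three images and sets their render translations, using per-image compact offset variables that default to zero.
# The event graph shows its SetMobileStyle call unconnected to PreConstruct or Construct.
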WidgetBlueprint /QuestCategories/Items/QuestRewards/QuestRewards_S39_MimosaShark/WBP_QuestRewards_Panel_S39_MimosaShark
graphs [
 EventGraph {nodes [
   SetMobileStyle() {is=title K2Node_CallFunction}
   Event PreConstruct()
   Event Construct()
  ]}
 SetMobileStyle {nodes [
   SetMobileStyle() {is=title K2Node_FunctionEntry}
   > Branch(Condition=IsUsingCompactUI())
   n1 = Get Image_Placeholder_03() -> (Image_Placeholder_03:Image Object Reference)
   n2 = SlotAsOverlaySlot(Widget=@n1)
   n3 = Get Image_Placeholder_02() -> (Image_Placeholder_02:Image Object Reference)
   n4 = SlotAsOverlaySlot(Widget=@n3)
   n5 = Get Image_Placeholder_01() -> (Image_Placeholder_01:Image Object Reference)
   n6 = SlotAsOverlaySlot(Widget=@n5)
   > SetHorizontalAlignment(self=[@n2, @n4, @n6], InHorizontalAlignment=HAlign_Center)
   > SetVerticalAlignment(self=[@n2, @n4, @n6], InVerticalAlignment=VAlign_Center)
   > SetRenderTranslation(self=[@n5, @n3, @n1])
   > SetRenderTranslation(self=@n5, Translation=Get Image1_CompactTranslation() -> (Image1_CompactTranslation:Vector 2D Structure))
   > SetRenderTranslation(self=@n3, Translation=Get Image2_CompactTranslation() -> (Image2_CompactTranslation:Vector 2D Structure))
   > SetRenderTranslation(self=@n1, Translation=Get Image3_CompactTranslation() -> (Image3_CompactTranslation:Vector 2D Structure))
  ]}
]
variables [
 Image1_CompactTranslation {type=Vector2D, default=<0 0>}
 Image2_CompactTranslation {type=Vector2D, default=<0 0>}
 Image3_CompactTranslation {type=Vector2D, default=<0 0>}
]
widgets {
 info {inheritedWidgetCount=0, parentClass=&UserWidget, rootWidgetClass=&SizeBox, widgetCount=18}
 widgets [
  SizeBox_0 {
   widgetClassPath=&SizeBox
   widget=SizeBox_0
   properties {bOverride_HeightOverride=true, bOverride_WidthOverride=true, heightOverride=900, widthOverride=850}
  }
  Overlay_0 {widgetClassPath=&Overlay, parent=SizeBox_0, widget=Overlay_0, slot {class=SizeBoxSlot}}
  Image_122 {
   widgetClassPath=&Image
   bIsVariable=true
   parent=Overlay_0
   widget=Image_122
   properties {colorAndOpacity=<rgba 0.0233 0.053 1 0>}
   slot {class=OverlaySlot, properties {horizontalAlignment=HAlign_Fill, verticalAlignment=VAlign_Fill}}
  }
  Image_BackgroundGlow {
   widgetClassPath=&Image
   bIsVariable=true
   parent=Overlay_0
   widget=Image_BackgroundGlow
   properties {
    brush {
     bIsDynamicallyLoaded=false
     drawAs=Image
     imageType=FullColor
     mirroring=NoMirror
     resourceName=None
     resourceObject=/QuestCategories/Items/QuestRewards/QuestRewards_S39_MimosaShark/Materials/MI_UI_QuestRewards_BackgroundGlow_MimosaShark
     tiling=NoTile
     imageSize=<1025 1025>
     margin=<ltrb 0 0 0 0>
     outlineSettings {
      bUseBrushTransparency=false
      roundingType=HalfHeightRadius
      width=0
      color {colorUseRule=UseColor_Specified, specifiedColor=<rgba 0 0 0 0>}
      cornerRadii=<0 0 0 0>
     }
     tintColor {colorUseRule=UseColor_Specified, specifiedColor=<rgba 0.9573 1 0.5003 1>}
     uVRegion {bIsValid=false, max=<0 0>, min=<0 0>}
    }
    colorAndOpacity=<rgba 1 1 1 0.2>
    renderTransform {angle=0, scale=<1.65 1.75>, shear=<0 0>, translation=<0 -63>}
   }
   slot {class=OverlaySlot, properties {horizontalAlignment=HAlign_Center, verticalAlignment=VAlign_Center}}
  }
  Image_ForegroundGlow {
   widgetClassPath=&Image
   bIsVariable=true
   parent=Overlay_0
   widget=Image_ForegroundGlow
   properties {
    brush {
     bIsDynamicallyLoaded=false
     drawAs=Image
     imageType=FullColor
     mirroring=NoMirror
     resourceName=None
     resourceObject=/QuestCategories/Items/QuestRewards/QuestRewards_S39_MimosaShark/Materials/MI_UI_QuestRewards__ForegroundGlow_MimosaShark
     tiling=NoTile
     imageSize=<1024 1024>
     margin=<ltrb 0 0 0 0>
     outlineSettings {
      bUseBrushTransparency=false
      roundingType=HalfHeightRadius
      width=0
      color {colorUseRule=UseColor_Specified, specifiedColor=<rgba 0 0 0 0>}
      cornerRadii=<0 0 0 0>
     }
     tintColor {colorUseRule=UseColor_Specified, specifiedColor=<rgba 1 0.9342 0 1>}
     uVRegion {bIsValid=false, max=<0 0>, min=<0 0>}
    }
    colorAndOpacity=<rgba 1 1 1 0.08>
    renderTransform {angle=0, scale=<1.7 2>, shear=<0 0>, translation=<1 0>}
   }
   slot {class=OverlaySlot, properties {horizontalAlignment=HAlign_Right, verticalAlignment=VAlign_Center}}
  }
  Image_Particles {
   widgetClassPath=&Image
   bIsVariable=true
   parent=Overlay_0
   widget=Image_Particles
   properties {
    brush {
     bIsDynamicallyLoaded=false
     drawAs=Image
     imageType=FullColor
     mirroring=NoMirror
     resourceName=None
     resourceObject=/FortniteCoreUI/UI/Quests/Materials/M_UI_RewardPanel_Particles
     tiling=NoTile
     imageSize=<64 4>
     margin=<ltrb 0 0 0 0>
     outlineSettings {
      bUseBrushTransparency=false
      roundingType=HalfHeightRadius
      width=0
      color {colorUseRule=UseColor_Specified, specifiedColor=<rgba 0 0 0 0>}
      cornerRadii=<0 0 0 0>
     }
     tintColor {colorUseRule=UseColor_Specified, specifiedColor=<rgba 1 0.9226 0 1>}
     uVRegion {bIsValid=false, max=<0 0>, min=<0 0>}
    }
    colorAndOpacity=<rgba 1 1 1 0.02>
    renderTransform {angle=90, scale=<0.65 0.68>, shear=<0 0>, translation=<1 0>}
   }
   slot {class=OverlaySlot, properties {horizontalAlignment=HAlign_Fill, verticalAlignment=VAlign_Fill}}
  }
  Image_Placeholder_03 {
   widgetClassPath=&Image
   bIsVariable=true
   parent=Overlay_0
   widget=Image_Placeholder_03
   properties {
    brush {
     bIsDynamicallyLoaded=false
     drawAs=Image
     imageType=FullColor
     mirroring=NoMirror
     resourceName=None
     resourceObject=/FortniteCoreUI/UI/Quests/Textures/T_UI_Shiny_llama__SquadEvent_Quests
     tiling=NoTile
     imageSize=<386 575>
     margin=<ltrb 0 0 0 0>
     outlineSettings {
      bUseBrushTransparency=false
      roundingType=HalfHeightRadius
      width=0
      color {colorUseRule=UseColor_Specified, specifiedColor=<rgba 0 0 0 0>}
      cornerRadii=<0 0 0 0>
     }
     tintColor {colorUseRule=UseColor_Specified, specifiedColor=<rgba 1 1 1 1>}
     uVRegion {bIsValid=false, max=<0 0>, min=<0 0>}
    }
    colorAndOpacity=<rgba 1 1 1 0.3>
    renderTransform {angle=0, scale=<0.95 0.95>, shear=<0 0>, translation=<0 -101>}
   }
   slot {class=OverlaySlot, properties {horizontalAlignment=HAlign_Center, verticalAlignment=VAlign_Center}}
  }
  Image_Placeholder_02 {
   widgetClassPath=&Image
   bIsVariable=true
   parent=Overlay_0
   widget=Image_Placeholder_02
   properties {
    brush {
     bIsDynamicallyLoaded=false
     drawAs=Image
     imageType=FullColor
     mirroring=NoMirror
     resourceName=None
     resourceObject=/QuestCategories/Items/QuestRewards/QuestRewards_S39_MimosaShark/Materials/T_UI_QuestRewards_MimosaShark
     tiling=NoTile
     imageSize=<512 512>
     margin=<ltrb 0 0 0 0>
     outlineSettings {
      bUseBrushTransparency=false
      roundingType=HalfHeightRadius
      width=0
      color {colorUseRule=UseColor_Specified, specifiedColor=<rgba 0 0 0 0>}
      cornerRadii=<0 0 0 0>
     }
     tintColor {colorUseRule=UseColor_Specified, specifiedColor=<rgba 1 1 1 1>}
     uVRegion {bIsValid=false, max=<0 0>, min=<0 0>}
    }
    colorAndOpacity=<rgba 1 1 1 0>
    renderTransform {angle=0, scale=<0.8 0.8>, shear=<0 0>, translation=<-96 -302>}
   }
   slot {class=OverlaySlot, properties {horizontalAlignment=HAlign_Center, verticalAlignment=VAlign_Center}}
  }
  Image_Placeholder_01 {
   widgetClassPath=&Image
   bIsVariable=true
   parent=Overlay_0
   widget=Image_Placeholder_01
   properties {
    brush {
     bIsDynamicallyLoaded=false
     drawAs=Image
     imageType=FullColor
     mirroring=NoMirror
     resourceName=None
     resourceObject=/FortniteCoreUI/UI/Quests/Textures/T_UI_Quest_Pickaxe_AkinboEnvoy
     tiling=NoTile
     imageSize=<512 512>
     margin=<ltrb 0 0 0 0>
     outlineSettings {
      bUseBrushTransparency=false
      roundingType=HalfHeightRadius
      width=0
      color {colorUseRule=UseColor_Specified, specifiedColor=<rgba 0 0 0 0>}
      cornerRadii=<0 0 0 0>
     }
     tintColor {colorUseRule=UseColor_Specified, specifiedColor=<rgba 1 1 1 1>}
     uVRegion {bIsValid=false, max=<0 0>, min=<0 0>}
    }
    renderTransform {angle=0, scale=<1 1>, shear=<0 0>, translation=<-7 -35>}
   }
   slot {class=OverlaySlot, properties {horizontalAlignment=HAlign_Center, verticalAlignment=VAlign_Center}}
  }
  VerticalBox_0 {
   widgetClassPath=&VerticalBox
   parent=Overlay_0
   widget=VerticalBox_0
   slot {class=OverlaySlot, properties {horizontalAlignment=HAlign_Fill, verticalAlignment=VAlign_Bottom, padding=<ltrb 50 20 50 20>}}
  }
  HorizontalBox {widgetClassPath=&HorizontalBox, parent=VerticalBox_0, widget=HorizontalBox, slot {class=VerticalBoxSlot}}
  Overlay_1 {
   widgetClassPath=&Overlay
   parent=HorizontalBox
   widget=Overlay_1
   slot {class=HorizontalBoxSlot, properties {padding=<ltrb 0 0 10 0>}}
  }
  Back {
   widgetClassPath=&Image
   bIsVariable=true
   parent=Overlay_1
   widget=Back
   properties {brush {
     bIsDynamicallyLoaded=false
     drawAs=Image
     imageType=FullColor
     mirroring=NoMirror
     resourceName=None
     resourceObject=/QuestCategories/Items/QuestRewards/QuestRewards_S39_MimosaShark/Materials/MI_UI_QuestRewards_InfoBackground_MimosaShark
     tiling=NoTile
     imageSize=<38 38>
     margin=<ltrb 0 0 0 0>
     outlineSettings {
      bUseBrushTransparency=false
      roundingType=HalfHeightRadius
      width=0
      color {colorUseRule=UseColor_Specified, specifiedColor=<rgba 0 0 0 0>}
      cornerRadii=<0 0 0 0>
     }
     tintColor {colorUseRule=UseColor_Specified, specifiedColor=<rgba 1 1 1 1>}
     uVRegion {bIsValid=false, max=<0 0>, min=<0 0>}
    }}
   slot {class=OverlaySlot, properties {horizontalAlignment=HAlign_Center, verticalAlignment=VAlign_Center}}
  }
  SharingQuest {
   widgetClassPath=&Image
   bIsVariable=true
   parent=Overlay_1
   widget=SharingQuest
   properties {brush {
     bIsDynamicallyLoaded=false
     drawAs=Image
     imageType=FullColor
     mirroring=NoMirror
     resourceName=None
     resourceObject=/Game/Athena/UI/Challenges/Art/T_UI_PartyAssist_64
     tiling=NoTile
     imageSize=<32 32>
     margin=<ltrb 0 0 0 0>
     outlineSettings {
      bUseBrushTransparency=false
      roundingType=HalfHeightRadius
      width=0
      color {colorUseRule=UseColor_Specified, specifiedColor=<rgba 0 0 0 0>}
      cornerRadii=<0 0 0 0>
     }
     tintColor {colorUseRule=UseColor_Foreground, specifiedColor=<rgba 1 1 1 1>}
     uVRegion {bIsValid=false, max=<0 0>, min=<0 0>}
    }}
   slot {class=OverlaySlot, properties {horizontalAlignment=HAlign_Center, verticalAlignment=VAlign_Center, padding=<ltrb 10 8 8 8>}}
  }
  Text_Disclaimer_1 {
   widgetClassPath=&CommonTextBlock
   parent=HorizontalBox
   widget=Text_Disclaimer_1
   properties {
    autoWrapText=true
    renderOpacity=0.8
    text=Progress is shared between Squad members
    font {
     bForceMonospaced=false
     bMaterialIsStencil=false
     fontMaterial=null
     fontObject=&HeadingNow
     letterSpacing=0
     monospacedWidth=1
     size=13.5
     skewAmount=0
     typefaceFontName=74 Regular
     outlineSettings {
      bApplyOutlineToDropShadows=false
      bMiteredCorners=false
      bSeparateFillAlpha=false
      outlineMaterial=null
      outlineSize=0
      outlineColor=<rgba 0 0 0 1>
     }
    }
    shadowOffset=<0 0>
   }
   slot {class=HorizontalBoxSlot, properties {verticalAlignment=VAlign_Center, size {sizeRule=Fill, value=1}}}
  }
  Border_122 {
   widgetClassPath=&Border
   parent=VerticalBox_0
   widget=Border_122
   properties {
    background {
     bIsDynamicallyLoaded=false
     drawAs=Image
     imageType=FullColor
     mirroring=NoMirror
     resourceName=None
     resourceObject=&MI_UI_QuestRewards_InfoBackground_MimosaShark
     tiling=NoTile
     imageSize=<32 32>
     margin=<ltrb 0 0 0 0>
     outlineSettings {
      bUseBrushTransparency=false
      roundingType=HalfHeightRadius
      width=0
      color {colorUseRule=UseColor_Specified, specifiedColor=<rgba 0 0 0 0>}
      cornerRadii=<0 0 0 0>
     }
     tintColor {colorUseRule=UseColor_Specified, specifiedColor=<rgba 1 1 1 1>}
     uVRegion {bIsValid=false, max=<0 0>, min=<0 0>}
    }
    padding=<ltrb 50 20 50 20>
   }
   slot {class=VerticalBoxSlot}
  }
  HorizontalBox_0 {
   widgetClassPath=&HorizontalBox
   parent=Border_122
   widget=HorizontalBox_0
   slot {class=BorderSlot, properties {padding=<ltrb 50 20 50 20>}}
  }
  RichText_Description {
   widgetClassPath=/Script/UMG.RichTextBlock
   bIsVariable=true
   parent=HorizontalBox_0
   widget=RichText_Description
   properties {
    text="<img id=\"ShinyLlama\"/><YellowBold>Shiny Llamas</> <Description>can drop from elims, Rare Chests, Supply Drops, and Bosses while playing in</> <YellowBold>Squads with others in BR, OG, Reload, and Blitz.</> "
    textStyleSet=/Game/Athena/HUD/QuestIcons/DT_RichText_QuestTile
    wrapTextAt=700
    decoratorClasses [/Game/Athena/HUD/QuestIcons/BP_QuestTile_IconDecorator.BP_QuestTile_IconDecorator_C]
   }
   slot {class=HorizontalBoxSlot, properties {verticalAlignment=VAlign_Top, size {sizeRule=Fill, value=1}}}
  }
 ]
}
